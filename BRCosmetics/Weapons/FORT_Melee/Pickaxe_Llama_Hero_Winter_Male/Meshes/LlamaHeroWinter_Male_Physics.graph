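# Physics asset for the male Llama Hero Winter pickaxe mesh, defining capsule bodies and joint limits for its dynamic bones.
# The handle body anchors a neck_01, neck_02, head and jaw chain with very tight angular limits, allowing only slight movement.
# Four small wool bodies attach directly to the handle with looser swing and twist limits, so they can move more freely than the head chain.
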PhysicsAsset /BRCosmetics/Weapons/FORT_Melee/Pickaxe_Llama_Hero_Winter_Male/Meshes/LlamaHeroWinter_Male_Physics
bodies {
 Melee_TwoHanded_Handle [{shapeType=Capsule, length=95.6959, radius=2.3981, center=<0.0994 -1.5034 10.0425>, rotation=<rot -0.2609 -90 -0.1337>}]
 dyn_head [{shapeType=Capsule, length=17.6509, radius=21.7263, center=<0.1092 6.3836 8.8341>, rotation=<rot -41.2233 -90.7234 1.6298>}]
 dyn_jaw [{shapeType=Capsule, length=7.9674, radius=4.1063, center=<-0.2688 2.8483 -3.209>, rotation=<rot 0 -6.4646 -53.1015>}]
 dyn_neck_01 [{shapeType=Capsule, length=21.8799, radius=11.716, center=<-0.2114 -5.7333 3.1312>, rotation=<rot 0 -2.2555 -61.3437>}]
 dyn_neck_02 [{shapeType=Capsule, length=16.6612, radius=12.8383, center=<-0.7567 -4.6486 4.9215>, rotation=<rot 0 -9.0383 -61.9704>}]
 dyn_wool_01 [{shapeType=Capsule, length=0, radius=3.4575, center=<0.2602 -0.0134 0.0652>, rotation=<rot 2.9036 162.0507 -89.9999>}]
 dyn_wool_02 [{shapeType=Capsule, length=0, radius=3.5925, center=<0.1958 0.308 0.8112>, rotation=<rot -0.6378 97.6782 -89.2272>}]
 dyn_wool_03 [{shapeType=Capsule, length=0, radius=3.6903, center=<0.0615 0.1544 0.6487>, rotation=<rot 0 -178.5952 -86.6388>}]
 dyn_wool_04 [{shapeType=Capsule, length=0.1326, radius=3.2435, center=<0.0047 -0.0945 0.4014>, rotation=<rot -0.535 -177.9672 -90.3299>}]
}
constraints [
 {bone1Name=dyn_neck_01, bone2Name=Melee_TwoHanded_Handle, swing1LimitDegrees=1, swing2LimitDegrees=2, twistLimitDegrees=2}
 {bone1Name=dyn_neck_02, bone2Name=dyn_neck_01, swing1LimitDegrees=1, swing2LimitDegrees=2, twistLimitDegrees=2}
 {bone1Name=dyn_head, bone2Name=dyn_neck_02, swing1LimitDegrees=1, swing2LimitDegrees=2, twistLimitDegrees=2}
 {bone1Name=dyn_jaw, bone2Name=dyn_head, swing1LimitDegrees=1, swing2LimitDegrees=1, twistLimitDegrees=5}
 {bone1Name=dyn_wool_01, bone2Name=Melee_TwoHanded_Handle, swing1LimitDegrees=20, swing2LimitDegrees=20, twistLimitDegrees=20}
 {bone1Name=dyn_wool_02, bone2Name=Melee_TwoHanded_Handle, swing1LimitDegrees=20, swing2LimitDegrees=20, twistLimitDegrees=20}
 {bone1Name=dyn_wool_03, bone2Name=Melee_TwoHanded_Handle, swing1LimitDegrees=20, swing2LimitDegrees=20, twistLimitDegrees=20}
 {bone1Name=dyn_wool_04, bone2Name=Melee_TwoHanded_Handle, swing1LimitDegrees=20, swing2LimitDegrees=20, twistLimitDegrees=20}
]
