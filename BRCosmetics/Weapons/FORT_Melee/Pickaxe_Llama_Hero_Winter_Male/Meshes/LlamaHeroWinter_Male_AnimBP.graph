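# Animation Blueprint for the Llama Hero Winter (Male) pickaxe that adds secondary motion and a procedural neck and head wobble.
# The input pose passes through two handle-relative RigidBody simulations with opposing gravity, then additive bone transforms on two neck bones and the head.
# Each update, game-time sine waves drive one shared wobble rotation and three phase-offset, clamped scale pulses.
# Rotate_B, Rotate_C and several leftover math nodes are unused.
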AnimBlueprint /BRCosmetics/Weapons/FORT_Melee/Pickaxe_Llama_Hero_Winter_Male/Meshes/LlamaHeroWinter_Male_AnimBP
target_skeleton=&/LlamaHeroWinter_Male_Skeleton
graphs [
 AnimGraph {nodes [
   AnimGraphNode_SubInput_0 = Input Pose()
   n1 = RigidBody(ComponentPose=Local To Component(LocalPose=@AnimGraphNode_SubInput_0)) {
    bOverrideWorldGravity=true
    overridePhysicsAsset=&/LlamaHeroWinter_Male_Physics
    simulationSpace=BaseBoneSpace
    alphaBoolBlend {blendOption=Linear}
    baseBoneRef {boneName=Melee_TwoHanded_Handle}
    componentAppliedLinearAccClamp=<500 500 500>
    componentLinearAccScale=<3 3 3>
    componentLinearVelScale=<5 5 5>
    overrideWorldGravity {z=-150}
   }
   n2 = RigidBody(ComponentPose=@n1) {
    bOverrideWorldGravity=true
    overridePhysicsAsset=&/LlamaHeroWinter_Male_Physics1
    simulationSpace=BaseBoneSpace
    alphaBoolBlend {blendOption=Linear}
    baseBoneRef {boneName=Melee_TwoHanded_Handle}
    componentAppliedLinearAccClamp=<600 600 600>
    componentLinearAccScale=<3 3 3>
    componentLinearVelScale=<5 5 5>
    overrideWorldGravity {z=700}
   }
   n3 = Get Rotate_A() -> (Rotate_A:Rotator)
   n4 = "Transform (Modify) Bone - Bone: dyn_neck_01"(ComponentPose=@n2, Alpha=0.6, Translation="0,0,0", Rotation=@n3, Scale=Get Swell_A() -> (Swell_A:Vector)) {
    rotationMode=BMM_Additive
    rotationSpace=BCS_BoneSpace
    scaleMode=BMM_Additive
    scaleSpace=BCS_BoneSpace
    alphaBoolBlend {blendOption=Linear}
    boneToModify {boneName=dyn_neck_01}
   }
   n5 = "Transform (Modify) Bone - Bone: dyn_neck_02"(ComponentPose=@n4, Alpha=0.6, Translation="0,0,0", Rotation=@n3, Scale=Get Swell_B() -> (Swell_B:Vector)) {
    rotationMode=BMM_Additive
    rotationSpace=BCS_BoneSpace
    scaleMode=BMM_Additive
    scaleSpace=BCS_BoneSpace
    alphaBoolBlend {blendOption=Linear}
    boneToModify {boneName=dyn_neck_02}
   }
   n6 = "Transform (Modify) Bone - Bone: dyn_head"(ComponentPose=@n5, Alpha=0.6, Translation="0,0,0", Rotation=@n3, Scale=Get Swell_C() -> (Swell_C:Vector)) {
    rotationMode=BMM_Additive
    rotationSpace=BCS_BoneSpace
    scaleMode=BMM_Additive
    scaleSpace=BCS_BoneSpace
    alphaBoolBlend {blendOption=Linear}
    boneToModify {boneName=dyn_head}
   }
   Output Pose(Result=Component To Local(ComponentPose=@n6)) {layerGroup=None}
  ]}
 EventGraph {nodes [
   Event BlueprintUpdateAnimation()
   > n1 = Sequence!(execute) -> (then_0, then_1, then_2:Exec)
   n2 = GetGameTimeInSeconds()
   n3 = float + float(A=float * float(A=float * float(A=1, B=1), B=@n2), B=2) {is=title K2Node_CommutativeAssociativeBinaryOperator}
   n4 = MapRangeClamped(Value="Sin (Radians)"(A=float * float(A=@n3, B=0.4)), InRangeA=-1, InRangeB=-0.6, OutRangeA=0.98, OutRangeB=0.95)
   Set Swell_A(execute=@n1.then_1, Swell_A:Vector=MakeVector(X=@n4, Y=@n4, Z=@n4)) -> (Output_Get:Vector)
   n5 = float + float(A=float * float(A=float * float(A=1, B=1), B=@n2), B=4)
   n6 = MapRangeClamped(Value="Sin (Radians)"(A=float * float(A=@n5, B=0.4)), InRangeA=-1, InRangeB=-0.6, OutRangeA=0.96, OutRangeB=1.02)
   > Set Swell_B(Swell_B:Vector=MakeVector(X=@n6, Y=@n6, Z=@n6)) -> (Output_Get:Vector)
   n7 = float + float(A=float * float(A=float * float(A=1, B=1), B=@n2), B=6)
   n8 = "Sin (Radians)"(A=float * float(A=@n7, B=0.4))
   n9 = MapRangeClamped(Value=@n8, InRangeA=-1, InRangeB=-0.6, OutRangeA=1.02, OutRangeB=1.1)
   > Set Swell_C(Swell_C:Vector=MakeVector(X=MapRangeClamped(Value=@n8, InRangeA=-1, InRangeB=-0.6, OutRangeA=1.02, OutRangeB=1.2), Y=@n9, Z=@n9)) -> (Output_Get:Vector)
   n10 = GetGameTimeInSeconds()
   n11 = "Sin (Radians)"(A=float * float(A=float * float(A=0.2, B=8), B=@n10))
   n12 = "Sin (Radians)"(A=float * float(A=float * float(A=0.2, B=13), B=@n10))
   n13 = "Sin (Radians)"(A=float * float(A=float * float(A=0.2, B=10), B=@n10))
   Set Rotate_A(
    execute=@n1.then_0
    Rotate_A:Rotator=MakeRotator(Roll=float * float(A=@n13, B=0.6), Pitch=float * float(A=@n11, B=0.7), Yaw=float * float(A=@n12, B=0.5))
   ) -> (Output_Get:Rotator)
   TryGetPawnOwner()
   "Sin (Degrees)"!(A:double=@n7) -> (ReturnValue:double)
   n14 = "Sin (Radians)"(A=float + float(A=float * float(A=0, B=0), B=4))
   n15 = "Sin (Radians)"(A=float + float(A=float * float(A=0, B=0), B=4))
   n16 = float * float(A=1, B=2)
   n17 = "Sin (Radians)"(A=float + float(A=float * float(A=@n16, B=0), B=4))
   MakeVector(
    X=float + float(A=float * float(A=@n17, B=0.02), B=1)
    Y=float + float(A=float * float(A=@n15, B=0.02), B=1)
    Z=float + float(A=float * float(A=@n14, B=0.02), B=1)
   )
   n18 = float * float(A=1, B=2)
   n19 = "Sin (Radians)"(A=float + float(A=float * float(A=@n18, B=0), B=6))
   n20 = "Sin (Radians)"(A=float + float(A=float * float(A=@n18, B=0), B=6))
   n21 = "Sin (Radians)"(A=float + float(A=float * float(A=@n18, B=0), B=6))
   MakeVector(
    X=float + float(A=float * float(A=@n21, B=0.03), B=1.1)
    Y=float + float(A=float * float(A=@n20, B=0.03), B=1.1)
    Z=float + float(A=float * float(A=@n19, B=0.03), B=1.1)
   )
   "Sin (Degrees)"(A=@n3)
   "Sin (Degrees)"(A=@n5)
   Reroute Node(InputPin:Wildcard) -> (OutputPin:Wildcard)
   Reroute Node()
   Reroute Node()
   Reroute Node()
  ]}
]
variables [
 Rotate_A {type=Rotator}
 Rotate_B {type=Rotator}
 Rotate_C {type=Rotator}
 Swell_A {type=Vector}
 Swell_B {type=Vector}
 Swell_C {type=Vector}
]
